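# Niagara system for the Mangaoutline vehicle boost cosmetic, drawing exhaust cones, ribbons and a smoke trail at the car's boost sockets.
# Per-socket emitters sample skeletal mesh positions with car offsets and recolor by painted variant.
# Spawn groups vary cone and smoke sizes and feed render visibility tags across multiple renderers.
# Smoke adds per-variant fill colors and a black border color curve; ribbons and smoke inherit source movement.
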NiagaraSystem /VehicleCosmetics/Boosts/Boost_Mangaoutline/FX/NS_Boost_Mangaoutline
dependencies {
 usedDataInterfaces [
  &NiagaraDataInterfaceArrayColor
  &NiagaraDataInterfaceArrayFloat
  &NiagaraDataInterfaceCurve
  &NiagaraDataInterfaceColorCurve
  &NiagaraDataInterfaceSkeletalMesh
 ]
 usedDynamicInputs []
 usedModules [
  &self:SystemSpawnScript.SystemScriptSource.SystemScriptGraph.NiagaraNodeAssignment_0.SetVariables_#1
  &SystemState
 ]
}
emitters {
 NE_BoostCone_Mesh001 {
  emitterUpdateScript [
   &EmitterState(Life Cycle Mode=::NewEnumerator1, Loop Duration=1, UseLoopDelay=false)
   &VC_FilteredSocketCount()
   &SpawnRate(SpawnRate=150, Use Spawn Probability=false) {enabled=false}
   &SpawnRate(SpawnRate=16, Use Spawn Probability=false) {module=SpawnRate002}
   &SpawnRate(SpawnGroup=1, SpawnRate=10, Use Spawn Probability=false) {module=SpawnRate001}
  ]
  particleSpawnScript [
   &VC_PaintedVariantsOverride(
    Black=<rgba 0.25 0.25 0.25 1>
    BurntSienna=<rgba 0.6 0.32 0.12 1>
    Cobalt=<rgba 0.4 0.55 1 1>
    Crimson=<rgba 0.8 0.06 0.06 0.95>
    DefaultColor=<rgba 2.15 2.15 2.15 1>
    ForestGreen=<rgba 0 1 0 1>
    Grey=<rgba 0.6 0.6 0.6 1>
    Lime=<rgba 0.9167 1 0 1>
    Orange=<rgba 1 0.6667 0 1>
    Pink=<rgba 1 0.55 0.8364 1>
    Purple=<rgba 0.575 0.15 1 1>
    Saffron=<rgba 1 0.9656 0.175 1>
    SkyBlue=<rgba 0 0.9981 1.5 1>
   )
   &InitializeParticle(
    Color=$Particles.Color
    Lifetime=0.1
    Mesh Scale Max=&Multiply_VectorByFloat(
     Vector=/VehicleCosmetics/FX/Niagara/Modules/MergeEmitters/VC_VectorBySpawnGroup(
      SpawnGroup or ExecIndex:NiagaraInt32=0
      CustomInt:NiagaraInt32=0
      MaximumSpawnGroups:NiagaraInt32=2
      Vector Group 0:Vector3f=<0.8 0.75 0.8>
      Vector Group 1:Vector3f=<1 0.4 1>
      Vector Group 2:Vector3f=<0 0 0>
      Vector Group 3:Vector3f=<0 0 0>
      Vector Group 4:Vector3f=<0 0 0>
      Vector Group 5:Vector3f=<0 0 0>
     )
     Float=8
    )
    Mesh Scale Min=&Multiply_VectorByFloat(
     Vector=&VC_VectorBySpawnGroup(
      SpawnGroup or ExecIndex:NiagaraInt32=0
      CustomInt:NiagaraInt32=0
      MaximumSpawnGroups:NiagaraInt32=2
      Vector Group 0:Vector3f=<0.6 0.5 0.6>
      Vector Group 1:Vector3f=<0.6 0.3 0.6>
      Vector Group 2:Vector3f=<0 0 0>
      Vector Group 3:Vector3f=<0 0 0>
      Vector Group 4:Vector3f=<0 0 0>
      Vector Group 5:Vector3f=<0 0 0>
     )
     Float=8
    )
    Mesh Scale Mode=::NewEnumerator2
    Position Offset Coordinate Space=::Local
    Sprite Rotation Mode=::NewEnumerator2
    UsePositionOffset=true
    Write Material Random=true
   )
   &PartitionParticles(Number Of Partitions=$Emitter.PortCount, Particles Per Partition=2, Partition Mode:ENiagara_PartitionMode=::NewEnumerator1)
   &SkeletalMeshLocation(
    Bone / Socket Index=$Particles.PartitionParticles.Partition
    Bone / Socket Sampling Mode=::NewEnumerator5
    MaskSpawnGroup=false
    Mesh Sampling Type=::NewEnumerator6
    Sampled Position Offset=&AddVectorToPosition(Position=<-50 0 0>, Vector=Linked() {linkedVariable {name=User.DTCarOffset, type=NiagaraPosition}})
    Skeletal Mesh=$System.SkeletalMesh
   )
   &AddVelocity(Velocity=&RandomRangeVector(Minimum=<-500 -100 -50>, Maximum=<-500 100 50>)) {enabled=false}
   &InitialMeshOrientation(
    Mesh Orientation Mode=::NewEnumerator2
    Rotation=&RandomRangeVector(Minimum=<0 0 0>, Maximum=<1 0 0>)
    Use Rotation Vector=true
   )
   /SetVariables(Particles.RenderVisTag:NiagaraInt32=&SpawnGroup()) {isSetParameters=true}
  ]
  particleUpdateScript [
   &ParticleState()
   &SolveForcesAndVelocity(Clamp Velocity=false, Limit Acceleration=false) {enabled=false}
  ]
  rendererClasses [&NiagaraMeshRendererProperties, &NiagaraSpriteRendererProperties]
  renderers [
   {rendererClass=&NiagaraMeshRendererProperties, rendererIndex=0}
   {rendererClass=&NiagaraMeshRendererProperties, rendererIndex=1}
   {rendererClass=&NiagaraSpriteRendererProperties, rendererIndex=2}
  ]
 }
 NE_Ribbon {
  emitterUpdateScript [
   &EmitterState(Life Cycle Mode=::NewEnumerator1, Loop Duration=1, Recalculate Duration Each Loop=true, UseLoopDelay=false)
   &VC_FilteredSocketCount()
   &VC_RateQualityHandler(High Spawn Rate=60, Low Spawn Rate=30, Use Custom Rate?=true)
   &VC_SpawnRibbon_v1(Number of Ribbons=$Emitter.PortCount, SpawnRate=$Emitter.Spawn Rate)
  ]
  particleSpawnScript [
   &VC_PaintedVariantsOverride(
    Black=<rgba 0.2 0.2 0.2 3>
    BurntSienna=<rgba 0.3 0.1425 0.03 2>
    Cobalt=<rgba 0.3 0.375 0.75 2>
    Crimson=<rgba 0.45 0.045 0.045 2>
    DefaultColor=<rgba 1 1 1 2>
    ForestGreen=<rgba 0.18 0.45 0.045 2>
    Grey=<rgba 0.4 0.4 0.4 2>
    Lime=<rgba 0.4956 0.65 0.0325 2>
    Orange=<rgba 0.75 0.4148 0.0188 2>
    Pink=<rgba 0.6 0.3 0.525 2>
    Purple=<rgba 0.07 0 0.12 1>
    Saffron=<rgba 0.5 0.5 0.05 2>
    SkyBlue=<rgba 0.1125 0.4631 0.75 2>
    TitaniumWhite=<rgba 2.55 2.55 3 2>
   )
   &VC_InitializeRibbon_v1(
    Color=$Particles.Color
    Custom Offset=true
    Enable Skeletal Position=true
    Lifetime=1
    Local Vector Offset=&AddVectorToPosition(
     Position=/Niagara/DynamicInputs/TypeConversions/MakeCustomPositionFromBool(Bool:NiagaraBool=$User.bIsFrontendPreview, True Position:NiagaraPosition=<-55 0 0>, False Position:NiagaraPosition=<-55 0 0>)
     Vector=$User.DTCarOffset002
    )
    Offset Amount=1
    Ribbon Width=11.5
   )
   &AddVelocity(Velocity=<-1200 0 0>) {module=AddVelocity001}
  ]
  particleUpdateScript [
   &ParticleState()
   &InheritSourceMovement(
    Applied Position Scale=&Multiply_VectorByFloat(
     Vector=<1 1 1>
     Float=&FloatFromCurve(FloatCurve=DataInterface() {propertyValues {Curve {keys [
          {interpMode=RCIM_Cubic, leaveTangent=-0.0051, tangentMode=RCTM_Break, time=0, value=1}
          {arriveTangent=-0.247, interpMode=RCIM_Cubic, tangentMode=RCTM_Break, time=1, value=0.9}
         ]}}})
    )
    ApplyForce=false
    ApplyPosition=true
    ApplyPositionFalloff=false
   )
   &SolveForcesAndVelocity(Clamp Velocity=false, Limit Acceleration=false)
   &VC_UpdateRibbon_v1(Enable Skeletal Position=true)
   &DynamicMaterialParameters(Module.Index0Param0Write=true, Module.Index0Param1Write=true, Module.Index0Param2Write=true, Module.Index0Param3Write=true)
   &ScaleRibbonWidth(Ribbon Width Scale=&FloatFromCurve(FloatCurve=DataInterface() {propertyValues {Curve {keys [
         {arriveTangent=0.0579, interpMode=RCIM_Cubic, leaveTangent=0.0579, tangentMode=RCTM_User, time=0, value=0.5}
         {arriveTangent=-0.0764, interpMode=RCIM_Cubic, leaveTangent=-0.0764, tangentMode=RCTM_User, time=0.25, value=1}
         <at 1 0.35>
        ]}}}))
  ]
  rendererClasses [&NiagaraRibbonRendererProperties]
  renderers [{rendererClass=&NiagaraRibbonRendererProperties}]
 }
 NE_SmokeTrail {
  emitterSpawnScript [/SetVariables(
    Emitter.SmokeFill_Black:LinearColor=&SelectLinearColorFromArray(
     Array Sampling Mode=::NewEnumerator1
     Direct Array Index=$System.ColorVariantID
     Color Selection Array=DataInterface() {propertyValues {ColorData [
        <rgba 0 0 0 1>
        <rgba 0 0 0 1>
        <rgba 0 0 0 1>
        <rgba 0 0 0 1>
        <rgba 0 0 0 1>
        <rgba 0 0 0 1>
        <rgba 0 0 0 1>
        <rgba 0 0 0 1>
        <rgba 0 0 0 1>
        <rgba 0 0 0 1>
        <rgba 0 0 0 1>
        <rgba 0 0 0 1>
        <rgba 0 0 0 1>
        <rgba 0 0 0 1>
        <rgba 0.9 0.9 0.9 0.5>
       ]}}
     Random Seed=?
    )
    Emitter.SmokeFill_Color:LinearColor=&SelectLinearColorFromArray(
     Array Sampling Mode=::NewEnumerator1
     Direct Array Index=$System.ColorVariantID
     Color Selection Array=DataInterface() {propertyValues {ColorData [
        <rgba 1 1 1 1>
        <rgba 1 1 1 1>
        <rgba 0.025 0.025 0.025 1>
        <rgba 0.15 0.036 0.0075 500>
        <rgba 0.06 0.1053 0.4 500>
        <rgba 0.4 0 0 500>
        <rgba 0 0.45 0 500>
        <rgba 0.15 0.15 0.15 500>
        <rgba 0.2937 0.5 0.005 500>
        <rgba 0.65 0.2267 0 500>
        <rgba 0.4 0.16 0.4 500>
        <rgba 0.11 0.015 0.3 500>
        <rgba 1 1 0.05 500>
        <rgba 0 0.2994 0.45 500>
        <rgba 1 1 1 1>
       ]}}
     Random Seed=?
    )
   ) {isSetParameters=true}]
  emitterUpdateScript [
   &EmitterState()
   &VC_FilteredSocketCount()
   &SpawnRate(SpawnRate=55, Use Spawn Probability=false)
   &SpawnRate(SpawnGroup=1, SpawnRate=55, Use Spawn Probability=false) {module=SpawnRate001}
   &SpawnRate(SpawnGroup=2, SpawnRate=20, Use Spawn Probability=false) {module=SpawnRate002}
   &SpawnRate(SpawnGroup=3, SpawnRate=30, Use Spawn Probability=false) {module=SpawnRate003}
  ]
  particleSpawnScript [
   &VC_PaintedVariantsOverride(
    Black=<rgba 0.04 0.04 0.04 1>
    BurntSienna=<rgba 0.1 0.025 0 1>
    Cobalt=<rgba 0.06 0.1053 0.4 1>
    Crimson=<rgba 0.4 0 0 1>
    ForestGreen=<rgba 0 0.5 0 1>
    Grey=<rgba 0.15 0.15 0.15 1>
    Lime=<rgba 0.5833 1 0 1>
    Orange=<rgba 0.65 0.2267 0 1>
    Pink=<rgba 0.4 0.16 0.4 1>
    Purple=<rgba 0.11 0.015 0.3 1>
    Saffron=<rgba 0.8 0.8 0.04 1>
    SkyBlue=<rgba 0 0.2994 0.45 1>
    TitaniumWhite=<rgba 0 0 0 1>
   )
   &InitializeParticle(
    Color=$Particles.Color
    Lifetime Max=0.675
    Lifetime Min=0.4
    Lifetime Mode=::NewEnumerator1
    Position Offset Coordinate Space=::Local
    Sprite Rotation Mode=::NewEnumerator2
    Sprite Size Mode=::NewEnumerator3
    Uniform Sprite Size=&Multiply_Float(
     A=&RandomRangeFloat(Minimum=40, Maximum=80)
     B=&SelectFloatFromArray(
      Direct Array Index=&SpawnGroup()
      Float Selection Array=DataInterface() {propertyValues {FloatData [1.15,1.25,0.65,0.65]}}
     )
    )
    UsePositionOffset=true
    Write Material Random=true
   )
   &PartitionParticles(Number Of Partitions=$Emitter.PortCount)
   &SkeletalMeshLocation(
    Bone / Socket Index=$Particles.PartitionParticles.Partition
    Bone / Socket Sampling Mode=::NewEnumerator5
    MaskSpawnGroup=false
    Mesh Sampling Type=::NewEnumerator6
    Sampled Position Offset=&AddVectorToPosition(Position=<-140 0 -5>, Vector=Linked() {linkedVariable {name=User.DTCarOffset, type=NiagaraPosition}})
    Skeletal Mesh=$System.SkeletalMesh
   )
   &ShapeLocation(
    MaskSpawnGroup=false
    Override Seed=false
    Sphere Radius=&SelectFloatFromArray(
     Direct Array Index=&SpawnGroup()
     Float Selection Array=DataInterface() {propertyValues {FloatData [40,30,50,40]}}
    )
    UseOwnerScale=true
   )
   &AddVelocity(Velocity=&MakeCustomVectorFromBool(
     Bool=$User.bIsFrontendPreview
     True Vector=&RandomRangeVector(Minimum=<-1500 -250 -50>, Maximum=<-1500 250 200>)
     False Vector=&RandomRangeVector(Minimum=<-1100 -250 -50>, Maximum=<-1100 250 200>)
    ))
   /SetVariables(Particles.RenderVisTag:NiagaraInt32=&SpawnGroup()) {isSetParameters=true}
  ]
  particleUpdateScript [
   &ParticleState()
   &ScaleColor(
    Scale Alpha=&FloatFromCurve(
     FloatCurve=DataInterface() {propertyValues {Curve {
        defaultValue=1
        keys [
         {arriveTangent=-2.5933, interpMode=RCIM_Cubic, leaveTangent=-2.5933, tangentMode=RCTM_User, time=0.225, value=1}
         {interpMode=RCIM_Cubic, tangentMode=RCTM_User, time=1, value=0}
        ]
       }}}
     Scale Curve=50
    )
    ScaleA=true
    ScaleRGB=true
   )
   /SetVariables(Particles.ColorSprites_Border:LinearColor=&ColorFromCurve(ColorCurve=DataInterface() {propertyValues {
       AlphaCurve {keys [<at 0.3595 150 RCIM_Cubic>,<at 1 0 RCIM_Cubic>]}
       BlueCurve {keys [<at 0 0 RCIM_Cubic>,<at 1 0 RCIM_Cubic>]}
       GreenCurve {keys [<at 0 0 RCIM_Cubic>,<at 1 0 RCIM_Cubic>]}
       RedCurve {keys [<at 0 0 RCIM_Cubic>,<at 1 0 RCIM_Cubic>]}
      }})) {isSetParameters=true}
   &ScaleSpriteSize(Uniform Curve Sprite Scale=DataInterface() {propertyValues {Curve {keys [
        {interpMode=RCIM_Cubic, tangentMode=RCTM_User, time=0, value=0.3}
        {arriveTangent=-0.0701, interpMode=RCIM_Cubic, leaveTangent=-0.0701, tangentMode=RCTM_User, time=0.1, value=1}
        {arriveTangent=0.4533, interpMode=RCIM_Cubic, leaveTangent=0.4533, tangentMode=RCTM_User, time=0.35, value=1.6}
        {arriveTangent=0.1385, interpMode=RCIM_Cubic, leaveTangent=0.1385, tangentMode=RCTM_User, time=0.673, value=2.1339}
        {interpMode=RCIM_Cubic, tangentMode=RCTM_User, time=1, value=2.25}
       ]}}})
   &Drag(
    Drag=&FloatFromCurve(
     FloatCurve=DataInterface() {propertyValues {Curve {
        defaultValue=1
        keys [
         {arriveTangent=6.7224, interpMode=RCIM_Cubic, leaveTangent=6.7224, tangentMode=RCTM_User, time=0, value=0}
         {interpMode=RCIM_Cubic, tangentMode=RCTM_User, time=1, value=3}
        ]
       }}}
     Scale Curve=2.5
    )
    Use Linear Drag=true
    Use Rotational Drag=true
   )
   &AccelerationForce(Acceleration=&Multiply_VectorByFloat(
     Vector=&RandomRangeVector(Minimum=<0 -250 -400>, Maximum=<0 250 200>)
     Float=&FloatFromCurve(FloatCurve=DataInterface() {propertyValues {Curve {
         defaultValue=1
         keys [
          {arriveTangent=7.908, interpMode=RCIM_Cubic, leaveTangent=7.908, tangentMode=RCTM_User, time=0, value=0}
          {interpMode=RCIM_Cubic, tangentMode=RCTM_User, time=0.2, value=0.5}
         ]
        }}})
    ))
   &InheritSourceMovement(
    Applied Position Scale=&Multiply_VectorByFloat(
     Vector=<1 1 1>
     Float=&FloatFromCurve(FloatCurve=DataInterface() {propertyValues {Curve {
         defaultValue=1
         keys [{interpMode=RCIM_Cubic, tangentMode=RCTM_User, time=0, value=1}, {interpMode=RCIM_Cubic, tangentMode=RCTM_User, time=1, value=0}]
        }}})
    )
    ApplyForce=false
    ApplyPosition=true
    ApplyPositionFalloff=false
   )
   &GravityForce(Gravity=&RandomRangeVector(Minimum=<0 0 0>, Maximum=<0 0 150>)) {enabled=false}
   &SolveForcesAndVelocity(Clamp Velocity=false, Limit Acceleration=false)
  ]
  rendererClasses [&NiagaraSpriteRendererProperties]
  renderers [
   {rendererClass=&NiagaraSpriteRendererProperties}
   {rendererClass=&NiagaraSpriteRendererProperties, rendererIndex=1}
   {rendererClass=&NiagaraSpriteRendererProperties, rendererIndex=2}
   {rendererClass=&NiagaraSpriteRendererProperties, rendererIndex=3}
  ]
 }
}
summary {
 systemName=NS_Boost_Mangaoutline
 emitters [
  {emitterName=NE_SmokeTrail, rendererClasses [&NiagaraSpriteRendererProperties]}
  {emitterName=NE_Ribbon, rendererClasses [&NiagaraRibbonRendererProperties]}
  {
   emitterName=NE_BoostCone_Mesh001
   rendererClasses [&NiagaraMeshRendererProperties, &NiagaraSpriteRendererProperties]
  }
 ]
 userVariables [
  User.bIsFrontendPreview {type=NiagaraBool}
  User.bIsPainted {type=NiagaraBool}
  User.ColorVariantID {type=NiagaraInt32}
  User.DTCarOffset {type=NiagaraPosition, description="", defaultValue {}}
  User.DTCarOffset001 {type=Vector3f, description="", defaultValue {struct=Vector3f, value=<0 0 0>}}
  User.DTCarOffset002 {type=Vector3f, description="", defaultValue {struct=Vector3f, value=<0 0 0>}}
  User.GameModeID {type=NiagaraInt32}
  User.GlowOverride {type=NiagaraFloat, defaultValue=4, description=""}
  User.PaintedColor {type=LinearColor, description="", defaultValue {struct=LinearColor, value=<rgba 1 1 1 1>}}
  User.PaintedScalar {type=NiagaraFloat, defaultValue=1, description=""}
  User.PortCount {type=NiagaraInt32, defaultValue=1}
  User.SingleBoostPos {type=NiagaraPosition, description="", defaultValue {}}
 ]
}
